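# Niagara burst effect for the SmartHyena backpack cosmetic, named for reactivity, that puffs a few dust sprites from the backpack's skeletal mesh.
# Particle partitions pick the bone or socket and one of two X-mirrored velocities, throwing dust outward and upward.
# Drag, gravity and SubUV animation shape each puff, while time of day scales its faint alpha.
# A disabled Feathers emitter retains fluttering two-tone sprites, with wind intensity multiplied to zero.
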NiagaraSystem /BRCosmetics/Accessories/FORT_Backpacks/Backpack_F_MED_SmartHyena/FX/NS_Backpack_SmartHyena_Burst_Reactivity
dependencies {
 usedDataInterfaces [
  &NiagaraDataInterfaceArrayFloat3
  &NiagaraDataInterfaceCurve
  &NiagaraDataInterfaceArrayColor
  &NiagaraDataInterfaceVector2DCurve
 ]
 usedDynamicInputs []
 usedModules [
  &self:SystemSpawnScript.SystemScriptSource.SystemScriptGraph.NiagaraNodeAssignment_0.SetVariables_#1
  &SystemState
 ]
}
emitters {
 Dust {
  emitterUpdateScript [
   &EmitterState(Inactive Response=::NewEnumerator1, Life Cycle Mode=::NewEnumerator1, Loop Behavior=::NewEnumerator1, Loop Duration=2, UseLoopDelay=false)
   &SpawnBurst_Instantaneous(
    Spawn Count=&RandomRangeInt(Minimum=2, Maximum=3)
    Use Spawn Probability=false
    UseLoopCountLimit=false
   )
  ]
  particleSpawnScript [
   &PartitionParticles()
   /SetVariables(Particles.SpriteRotationRate=&RandomRangeFloat(Minimum=-0.35, Maximum=0.35)) {isSetParameters=true}
   &InitializeParticle(
    Color=<rgba 0.7917 0.6426 0.5401 0.9>
    Lifetime Max=2.5
    Lifetime Min=1.5
    Lifetime Mode=::NewEnumerator1
    Sprite Rotation Angle=&RandomRangeFloat(Minimum=0, Maximum=1)
    Sprite Rotation Mode=::NewEnumerator1
    Sprite Size Mode=::NewEnumerator1
    Uniform Sprite Size Max=75
    Uniform Sprite Size Min=35
    UsePositionOffset=false
    Write Material Random=true
   )
   &DynamicMaterialParameters(
    Index 0 Param 1=&RandomRangeFloat(Minimum=0, Maximum=1)
    Index 0 Param 2=&RandomRangeFloat(Minimum=0, Maximum=1)
    Module.Param0WriteEnabled=true
    Module.Param1WriteEnabled=true
    Module.Param2WriteEnabled=true
    Module.Param3WriteEnabled=true
   )
   &SkeletalMeshLocation(
    Bone / Socket Index=&Add_Integer(A=$Particles.PartitionParticles.Partition, B=1)
    Bone Sampling Mode=::NewEnumerator4
    MaskSpawnGroup=false
    Sampled Position Offset Coordinate Space:ENiagaraSkelMeshTransforms=::NewEnumerator2
    Skeletal Mesh=$System.SkeletalMesh
   )
   &ShapeLocation(MaskSpawnGroup=false, Override Seed=false, Sphere Radius=2.5, UseOwnerScale=true)
   &AddVelocity(
    Scale Added Velocity=<1 0 1>
    Velocity=&Multiply_Vector(
     A=&RandomRangeVector(Minimum=<0.25 0.5 0.05>, Maximum=<0.5 1 0.1>)
     B=&SelectVectorFromArray(
      Direct Array Index=$Particles.PartitionParticles.Partition
      Vector Selection Array=DataInterface() {propertyValues {FloatData [<-50 -50 80>, <50 -50 80>]}}
     )
    )
   )
   &CameraOffset(Camera Offset Amount=5)
  ]
  particleUpdateScript [
   &ParticleState()
   &AerodynamicDrag(Aerodynamic Drag=0.05, Aerodynamic Rotational Drag=0.1, Use Linear Drag=true, Use Rotational Drag=true)
   &GravityForce(Gravity=<0 0 -50>)
   &SolveForcesAndVelocity(Clamp Velocity=false, Limit Acceleration=false)
   &SpriteRotationRate(Delta Time:NiagaraFloat=$Particles.NormalizedAge, Rotation Rate=$Particles.SpriteRotationRate)
   &AlignSpriteToMeshOrientation(SetAlignment=true, SetFacing=true)
   &DynamicMaterialParameters(
    Index 0 Param 1=&RandomRangeFloat(Minimum=0, Maximum=1)
    Index 0 Param 2=&RandomRangeFloat(Minimum=0, Maximum=1)
    Index 1 Param 1=&RandomRangeFloat(Minimum=0.02, Maximum=0.05)
    Index 1 Param 2=&RandomRangeFloat(Minimum=0.02, Maximum=0.04)
    Index 1 Param 3=&RandomRangeFloat(Minimum=0, Maximum=0.1)
    Index 1 Param 4=&RandomRangeFloat(Minimum=0, Maximum=0.1)
    Module.Param0WriteEnabled=true
    Module.Param0WriteEnabled1=true
    Module.Param1WriteEnabled=true
    Module.Param1WriteEnabled1=true
    Module.Param2WriteEnabled=true
    Module.Param2WriteEnabled1=true
    Module.Param3WriteEnabled=true
    Module.Param3WriteEnabled1=true
    Write Parameter Index 1=true
   ) {module=DynamicMaterialParameters001}
   &ScaleColor(
    Scale Alpha=&Multiply_Float(
     A=&FloatFromCurve(FloatCurve=DataInterface() {propertyValues {Curve {keys [
          {arriveTangent=6.8592, interpMode=RCIM_Cubic, leaveTangent=6.8592, tangentMode=RCTM_User, time=0, value=0}
          {arriveTangent=0.6396, interpMode=RCIM_Cubic, leaveTangent=0.6396, time=0.1512, value=0.5516}
          {arriveTangent=-0.9119, interpMode=RCIM_Cubic, leaveTangent=-0.9119, tangentMode=RCTM_User, time=0.2958, value=0.1892}
          {interpMode=RCIM_Cubic, tangentMode=RCTM_User, time=1, value=0}
         ]}}})
     B=&FNTimeOfDayScalar(Night Scale=0.1, Morning Scale=0.075, Day Scale=0.075, Evening Scale=0.075)
    )
    Scale RGB=<1 1 1.2>
    ScaleA=true
    ScaleRGB=true
   ) {module=ScaleColor001}
   &ScaleSpriteSize(Scale Factor=&Vector2DFromFloat(Value=&FloatFromCurve(FloatCurve=DataInterface() {propertyValues {Curve {keys [
          {arriveTangent=2.6303, interpMode=RCIM_Cubic, leaveTangent=2.6303, tangentMode=RCTM_User, time=0, value=0}
          {interpMode=RCIM_Cubic, tangentMode=RCTM_User, time=0.3, value=1}
          <at 0.9 1>
          <at 1 0 RCIM_Cubic>
         ]}}}))) {module=ScaleSpriteSize001}
   &SubUVAnimation(Start Frame=16, SubUV Animation Mode:ENiagara_SubUVLookupMode=::NewEnumerator1)
  ]
  rendererClasses [&NiagaraSpriteRendererProperties]
  renderers [{rendererClass=&NiagaraSpriteRendererProperties}]
 }
 Feathers {
  bEnabled=false
  emitterUpdateScript [
   &EmitterState(Inactive Response=::NewEnumerator1, Life Cycle Mode=::NewEnumerator1, Loop Behavior=::NewEnumerator1, Loop Duration=3, UseLoopDelay=false)
   &SpawnBurst_Instantaneous(
    Spawn Count=&RandomRangeInt(Minimum=3, Maximum=5)
    Use Spawn Probability=false
    UseLoopCountLimit=false
   )
   &SpawnBurst_Instantaneous(
    Spawn Count=&RandomRangeInt(Minimum=2, Maximum=3)
    Spawn Time=0.075
    Use Spawn Probability=false
    UseLoopCountLimit=false
   ) {module=SpawnBurst_Instantaneous001}
   /SetVariables(
    Emitter.WindOscillateTime:NiagaraFloat=&Divide_Float(A=$Engine.Time, B=44)
    Emitter.WindRotateTime:NiagaraFloat=&Divide_Float(A=$Engine.Time, B=66)
   ) {isSetParameters=true}
   /SetVariables(
    Emitter.WindIntensity:NiagaraFloat=&Multiply_Float(A=&Sine(Normalized Angle=$Emitter.WindOscillateTime, Period=1, Scale=1, Bias=0), B=0)
    Emitter.WindVector:Vector3f=&MakeVector(
     X=&Add_Float(
      A=&Sine(Normalized Angle=$Emitter.WindRotateTime, Period=1, Scale=1, Bias=0)
      B=&Cosine(Normalized Angle=$Emitter.WindRotateTime, Period=1, Scale=1, Bias=0)
     )
     Y=&Subtract_Float(
      A=&Cosine(Normalized Angle=$Emitter.WindRotateTime, Period=1, Scale=1, Bias=0)
      B=&Sine(Normalized Angle=$Emitter.WindRotateTime, Period=1, Scale=1, Bias=0)
     )
     Z=0
    )
   ) {isSetParameters=true}
  ]
  particleSpawnScript [
   &PartitionParticles()
   &InitializeParticle(
    Color=&SelectLinearColorFromArray(
     Array Sampling Mode=::NewEnumerator0
     Direct Array Index=?
     Color Selection Array=DataInterface() {propertyValues {ColorData [<rgba 0.4621 0.7011 0.1946 1>, <rgba 0.8205 0.8333 0.48 1>]}}
     Random Seed=0
    )
    Lifetime Max=2.5
    Lifetime Min=1.5
    Lifetime Mode=::NewEnumerator1
    Sprite Rotation Mode=::NewEnumerator2
    Sprite Size Mode=::NewEnumerator1
    Uniform Sprite Size Max=11
    Uniform Sprite Size Min=6
    UsePositionOffset=false
    Write Material Random=true
   )
   &SkeletalMeshLocation(
    Bone / Socket Index=&Add_Integer(A=$Particles.PartitionParticles.Partition, B=1)
    Bone Sampling Mode=::NewEnumerator4
    MaskSpawnGroup=false
    Sampled Position Offset=<0 -5 0>
    Sampled Position Offset Coordinate Space:ENiagaraSkelMeshTransforms=::NewEnumerator2
    Skeletal Mesh=$System.SkeletalMesh
   )
   &ShapeLocation(
    Handle Radius=5
    Large Radius=5
    MaskSpawnGroup=false
    Non Uniform Scale=<2 0 1>
    Override Seed=false
    Shape Primitive=::NewEnumerator3
    Surface Distribution=1
    UseOwnerScale=true
   )
   &AddVelocity(
    Cone Angle=90
    Cone Axis=<-0.1 -0.5 0.5>
    Distribution Along Cone Axis=0
    Module.UseConeFalloff=true
    Module.UseVelDistribution=false
    Override Seed=false
    Speed Falloff From Cone Axis=0.25
    Velocity Mode=::NewEnumerator2
    Velocity Speed=&RandomRangeFloat(Minimum=50, Maximum=75)
   ) {module=AddVelocity002}
   &AddVelocity(Velocity=<0 0 1>, Velocity Speed Scale=200)
   &ApplyInitialForces()
  ]
  particleUpdateScript [
   &ParticleState()
   &GravityForce(Gravity=&RandomRangeVector(Minimum=<0 0 -980>, Maximum=<0 0 -980>))
   &AccelerationForce(Acceleration=<0 0 333>)
   &PointForce(Force Falloff Distance=333, Force Strength=250, Module.Use Falloff=true, Module.UseLinearForce=false)
   &WindForce(Wind Speed=$Emitter.WindVector, Wind Speed Scale=$Emitter.WindIntensity)
   &AerodynamicDrag(
    Drag Shape Curve=DataInterface() {propertyValues {Curve {keys [
        {interpMode=RCIM_Cubic, tangentMode=RCTM_User, time=0, value=0.1}
        {arriveTangent=2.0799, interpMode=RCIM_Cubic, leaveTangent=2.0799, tangentMode=RCTM_User, time=1, value=1}
       ]}}}
    Lift Shape Curve=DataInterface() {propertyValues {Curve {keys [
        {interpMode=RCIM_Cubic, tangentMode=RCTM_User, time=0, value=0}
        {interpMode=RCIM_Cubic, tangentMode=RCTM_User, time=0.5, value=1}
        {tangentMode=RCTM_User, time=1, value=0}
       ]}}}
    Rotation Strength=2
    Use Linear Drag=true
    Use Rotational Drag=true
   )
   &SolveForcesAndVelocity(Clamp Velocity=false, Limit Acceleration=false)
   &SpriteRotationRate(Rotation Rate=&RandomRangeFloat(Minimum=-1, Maximum=1))
   &AlignSpriteToMeshOrientation(SetAlignment=true, SetFacing=true)
   &ScaleSpriteSize(
    Non-Uniform Curve Scale=&MakeVector2D(
     X=1
     Y=&Sine(Normalized Angle=$Emitter.Age, Period=&RandomRangeFloat(Minimum=2, Maximum=1), Scale=1, Bias=0.1)
    )
    Non-Uniform Curve Sprite Scale=DataInterface() {propertyValues {
      XCurve {keys [
        {arriveTangent=8.2479, interpMode=RCIM_Cubic, leaveTangent=8.2479, tangentMode=RCTM_User, time=0, value=0.5}
        <at 0.1 1>
        {interpMode=RCIM_Cubic, tangentMode=RCTM_User, time=0.7, value=1}
        {tangentMode=RCTM_User, time=1, value=0}
       ]}
      YCurve {keys [
        {arriveTangent=8.2479, interpMode=RCIM_Cubic, leaveTangent=8.2479, tangentMode=RCTM_User, time=0, value=0.5}
        <at 0.1 1>
        {interpMode=RCIM_Cubic, tangentMode=RCTM_User, time=0.7, value=1}
        {tangentMode=RCTM_User, time=1, value=0}
       ]}
     }}
    Scale Sprite Size Mode=::NewEnumerator3
   )
  ]
  rendererClasses [&NiagaraSpriteRendererProperties]
  renderers [{rendererClass=&NiagaraSpriteRendererProperties}]
 }
}
summary {
 systemName=NS_Backpack_SmartHyena_Burst_Reactivity
 emitters [
  {emitterName=Dust, rendererClasses [&NiagaraSpriteRendererProperties]}
  {bEnabled=false, emitterName=Feathers, rendererClasses [&NiagaraSpriteRendererProperties]}
 ]
 userVariables [
  User.BaseColor {type=LinearColor, description="", defaultValue {struct=LinearColor, value=<rgba 0.8632 0.5122 0.3237 1>}}
  User.bIsFrontEnd {type=NiagaraBool}
  User.bIsFrontEndPreview {type=NiagaraBool}
  User.FrontEndToggle {type=NiagaraBool}
  User.Reactivity {type=NiagaraBool}
  User.Skeletal Mesh {
   type=NiagaraDataInterfaceSkeletalMesh
   defaultValue {
    struct=NiagaraExt_VariableValue_DataInterface
    value {dataInterface=&self:NiagaraDataInterfaceSkeletalMesh_0}
   }
  }
  User.SpawnRateScale {type=NiagaraFloat, defaultValue=1, description=""}
  User.SphereSize {type=NiagaraFloat, defaultValue=1000, description=""}
 ]
}
