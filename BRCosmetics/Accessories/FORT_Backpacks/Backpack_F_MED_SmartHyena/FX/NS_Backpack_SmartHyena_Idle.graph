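# Niagara idle effect for the SmartHyena backpack cosmetic: a flickering orange flare with accompanying light.
# A single Flare emitter bursts two looping particles that follow the skeletal mesh at an offset and draw through sprite, light and component renderers.
# Separate waveforms pulse the sprite alpha and component light intensity, with front-end modifiers switching between gameplay and menu values, dimming the light in the front end.
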NiagaraSystem /BRCosmetics/Accessories/FORT_Backpacks/Backpack_F_MED_SmartHyena/FX/NS_Backpack_SmartHyena_Idle
dependencies {
 usedDataInterfaces []
 usedDynamicInputs []
 usedModules [
  &self:SystemSpawnScript.SystemScriptSource.SystemScriptGraph.NiagaraNodeAssignment_0.SetVariables_#1
  &SystemState
 ]
}
emitters {Flare {
  emitterUpdateScript [
   &EmitterState(Life Cycle Mode=::NewEnumerator1, Loop Behavior=::NewEnumerator1, Loop Duration Mode=::NewEnumerator1, UseLoopDelay=false)
   &SpawnBurst_Instantaneous(Spawn Count=2, Use Spawn Probability=false, UseLoopCountLimit=false)
  ]
  particleSpawnScript [&InitializeParticle(
    Color=<rgba 1 0.3333 0 1>
    Lifetime Max=0.6
    Lifetime Min=0.4
    Lifetime Mode=::NewEnumerator1
    Sprite Rotation Mode=::NewEnumerator2
    Sprite Size Mode=::NewEnumerator3
    Uniform Sprite Size=50
    UsePositionOffset=false
    Write Material Random=true
   ) {module=InitializeParticle001}]
  particleUpdateScript [
   &ParticleState(
    Kill Particles When Lifetime Has Elapsed=false
    Let Infinitely Lived Particles Die When Emitter Deactivates=true
    Loop Particles Lifetime=true
   )
   &SkeletalMeshLocation(
    Bone Sampling Mode=::NewEnumerator4
    MaskSpawnGroup=false
    Sampled Position Offset=<0 0 -12.5>
    Sampled Position Offset Coordinate Space:ENiagaraSkelMeshTransforms=::NewEnumerator2
    Skeletal Mesh=$System.SkeletalMesh
   )
   &FN_Light_Attributes(Create Exponent Attribute=true, Exponent=4)
   /SetVariables(
    Particles.LightComponentIntensity:NiagaraFloat=&Multiply_Float(
     A=70
     B=&FN_FrontEndFloatModifier(
      Gameplay=&Add_Float(A=0.5, B=&Waveform("Waveform [1]"=::NewEnumerator9, "[1] Frequency"=8, "[1] Amplitude Min/Max"=<0 1>))
      Front End=&Multiply_Float(
       A=&Add_Float(A=0.5, B=&Waveform("Waveform [1]"=::NewEnumerator9, "[1] Frequency"=8, "[1] Amplitude Min/Max"=<0 1>))
       B=0.7
      )
     )
    )
    Particles.Position_LightComponent:NiagaraPosition=&AddVectorToPosition(Position=$Particles.Position, Vector=<0 0 -1>)
   ) {isSetParameters=true}
   &ScaleColor(
    Scale Alpha=&FN_FrontEndFloatModifier(
     Gameplay=&Add_Float(A=0.5, B=&Waveform("Waveform [1]"=::NewEnumerator9, "[1] Frequency"=9, "[1] Amplitude Min/Max"=<0 1>))
     Front End=&Multiply_Float(
      A=&Add_Float(A=0.5, B=&Waveform("Waveform [1]"=::NewEnumerator9, "[1] Frequency"=9, "[1] Amplitude Min/Max"=<0 1>))
      B=1
     )
    )
    ScaleA=true
    ScaleRGB=true
   ) {module=ScaleColor001}
  ]
  rendererClasses [
   &NiagaraSpriteRendererProperties
   &NiagaraLightRendererProperties
   &NiagaraComponentRendererProperties
  ]
  renderers [
   {rendererClass=&NiagaraSpriteRendererProperties}
   {rendererClass=&NiagaraLightRendererProperties, rendererIndex=1}
   {rendererClass=&NiagaraComponentRendererProperties, rendererIndex=2}
  ]
 }}
summary {
 systemName=NS_Backpack_SmartHyena_Idle
 emitters [{
   emitterName=Flare
   rendererClasses [
    &NiagaraSpriteRendererProperties
    &NiagaraLightRendererProperties
    &NiagaraComponentRendererProperties
   ]
  }]
 userVariables [
  User.bIsFrontEnd {type=NiagaraBool}
  User.bIsFrontEndPreview {type=NiagaraBool}
  User.FrontEndToggle {type=NiagaraBool}
  User.Reactivity {type=NiagaraBool}
  User.Skeletal Mesh {
   type=NiagaraDataInterfaceSkeletalMesh
   defaultValue {
    struct=NiagaraExt_VariableValue_DataInterface
    value {dataInterface=&self:NiagaraDataInterfaceSkeletalMesh_0}
   }
  }
 ]
}
