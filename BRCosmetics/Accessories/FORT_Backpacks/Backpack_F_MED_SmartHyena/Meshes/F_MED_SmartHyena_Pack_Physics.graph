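# Physics asset for the SmartHyena back bling cosmetic, a bird with a light box, defining bodies for secondary motion and collision.
# A seven-segment tail chain hangs from BirdMain through BirdTailStart, stiff at the base and looser toward the tip.
# Head feathers have tight limits to the head, and the light box swings on one limited axis from BirdMain.
# Wings, wing feathers, pelvis and root have bodies but no constraints.
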PhysicsAsset /BRCosmetics/Accessories/FORT_Backpacks/Backpack_F_MED_SmartHyena/Meshes/F_MED_SmartHyena_Pack_Physics
bodies {
 BirdMain_M [{shapeType=Capsule, length=1.7199, radius=7.7801, shapeName=BirdMain_M_capsule, center=<6.1856 3.3141 0>, rotation=<rot 90 -86.9872 -54.1903>}]
 BirdTail1_M [{shapeType=Capsule, length=6.2564, radius=1.5229, shapeName=BirdTail1_M_capsule, center=<4.1749 0 0>, rotation=<rot 90 -19.471 -19.4713>}]
 BirdTail2_M [{
   shapeType=Capsule
   length=6.2564
   radius=1.5229
   shapeName=BirdTail1_M_capsule
   center=<4.1717 -0.0467 0>
   rotation=<rot 90 -144.2933 -140.4295>
  }]
 BirdTail3_M [{shapeType=Capsule, length=7.7716, radius=1.6062, shapeName=BirdTail1_M_capsule, center=<4.3425 0 0>, rotation=<rot 90 -19.471 -19.4713>}]
 BirdTail4_M [{shapeType=Capsule, length=4.9562, radius=2.1544, shapeName=BirdTail1_M_capsule, center=<3.8674 0 0>, rotation=<rot 90 -19.471 -19.4713>}]
 BirdTail5_M [{shapeType=Capsule, length=3.1552, radius=2.7452, shapeName=BirdTail1_M_capsule, center=<2.7854 0 0>, rotation=<rot 90 -19.471 -19.4713>}]
 BirdTail6_M [{shapeType=Capsule, length=1.2997, radius=3.3017, shapeName=BirdTail1_M_capsule, center=<2.2747 0 0>, rotation=<rot 90 -19.471 -19.4713>}]
 BirdTail7_M [{shapeType=Capsule, length=1.1736, radius=3.4153, shapeName=BirdTail1_M_capsule, center=<1.5048 0 0>, rotation=<rot 90 -19.471 -19.4713>}]
 BirdTailStart_M [{shapeType=Capsule, length=2.5, radius=2.5, shapeName=BirdTailStart_M_capsule, center=<1.4803 0 0>, rotation=<rot -90 0 0.0004>}]
 HeadFeather_L [{
   shapeType=Box
   extentX=2.5
   extentY=0.5
   extentZ=4.5
   shapeName=HeadFeather_L_box
   center=<-1.2301 0.2564 -0.1615>
   rotation=<rot 0.0002 -30 0.0003>
  }]
 HeadFeather_R [{
   shapeType=Box
   extentX=2.5
   extentY=0.5
   extentZ=4.5
   shapeName=HeadFeather_L_box
   center=<1.2301 -0.2564 0.1615>
   rotation=<rot 0.0002 -30 -179.9997>
  }]
 LightBox_M [
  {shapeType=Capsule, length=11.5, radius=1.5, shapeName=LightBox_M_capsule, center=<0.1158 0 0>}
  {shapeType=Capsule, length=15.4419, radius=1.7494, shapeName=LightBox_M_capsule, center=<0.4848 -0.2815 0>}
  {shapeType=Capsule, length=7.6845, radius=6.4885, shapeName=LightBox_M_capsule, center=<13.8206 0 0>, rotation=<rot 90 180 179.9998>}
  {shapeType=Box, extentX=7.5, extentY=1.9913, extentZ=9.3382, shapeName=LightBox_M_box, center=<2.9982 0 0>}
 ]
 WingFeather_L [{
   shapeType=Capsule
   length=5
   radius=1.5
   shapeName=WingFeather_R_capsule
   center=<-2.6782 0.0004 0.496>
   rotation=<rot -85.2287 -179.7235 -0.2812>
  }]
 WingFeather_R [{
   shapeType=Capsule
   length=5
   radius=1.5
   shapeName=WingFeather_R_capsule
   center=<2.6782 -0.0004 -0.496>
   rotation=<rot -85.2287 -179.7235 179.7188>
  }]
 Wing_L [{shapeType=Capsule, length=4.5, radius=1.5, shapeName=Wing_R_capsule, center=<-0.2426 0.3338 -1.8698>, rotation=<rot -3.1574 0.265 175.1941>}]
 Wing_R [{shapeType=Capsule, length=4.5, radius=1.5, shapeName=Wing_R_capsule, center=<0.2426 -0.3338 1.8698>, rotation=<rot -3.1574 0.265 -4.8059>}]
 head_M [{shapeType=Capsule, length=1.4594, radius=8.0406, shapeName=head_M_capsule, center=<3.2625 1.5211 0>, rotation=<rot -90 -90 90.0004>}]
 pelvis [
  {shapeType=Capsule, length=7.6672, radius=14.9664, shapeName=pelvis_capsule, center=<-3.9648 0.2778 0>, rotation=<rot 0 0.0003 -0.0001>}
  {shapeType=Box, extentX=11.5, extentY=12.5, extentZ=18.5, shapeName=pelvis_box, center=<10.188 -1.4746 0>}
  {
   shapeType=Box
   extentX=38.5546
   extentY=10.3572
   extentZ=32.2326
   shapeName=pelvis_box
   center=<-23.6318 -6.0091 0>
   rotation=<rot 0 14.8057 -0.0001>
  }
 ]
 root [{shapeType=Box, extentX=29.5946, extentY=18.6878, extentZ=67.8854, shapeName=root_box, center=<0 15.3657 -6.778>}]
}
constraints [
 {bone1Name=HeadFeather_L, bone2Name=head_M, swing1LimitDegrees=4, swing2LimitDegrees=4, twistLimitDegrees=1}
 {bone1Name=HeadFeather_R, bone2Name=head_M, swing1LimitDegrees=4, swing2LimitDegrees=4, twistLimitDegrees=1}
 {
  bone1Name=LightBox_M
  bone2Name=BirdMain_M
  swing1LimitDegrees=8
  swing2LimitDegrees=45
  swing2Motion=Locked
  twistLimitDegrees=45
  twistMotion=Locked
 }
 {bone1Name=BirdTail2_M, bone2Name=BirdTail1_M, swing1LimitDegrees=3, swing2LimitDegrees=3, twistLimitDegrees=1}
 {bone1Name=BirdTail3_M, bone2Name=BirdTail2_M, swing1LimitDegrees=4, swing2LimitDegrees=4, twistLimitDegrees=3}
 {bone1Name=BirdTail4_M, bone2Name=BirdTail3_M, swing1LimitDegrees=15, swing2LimitDegrees=12, twistLimitDegrees=3}
 {bone1Name=BirdTail5_M, bone2Name=BirdTail4_M, swing1LimitDegrees=15, swing2LimitDegrees=15, twistLimitDegrees=3}
 {bone1Name=BirdTail6_M, bone2Name=BirdTail5_M, swing1LimitDegrees=10, swing2LimitDegrees=20, twistLimitDegrees=3, twistMotion=Locked}
 {bone1Name=BirdTail7_M, bone2Name=BirdTail6_M, swing1LimitDegrees=10, swing2LimitDegrees=20, twistLimitDegrees=3, twistMotion=Locked}
 {bone1Name=BirdTail1_M, bone2Name=BirdTailStart_M, swing1LimitDegrees=2, swing2LimitDegrees=2, twistLimitDegrees=2}
 {bone1Name=BirdTailStart_M, bone2Name=BirdMain_M, swing1LimitDegrees=1, swing2LimitDegrees=1, twistLimitDegrees=1}
]
