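# Physics asset for the SmartHyena backpack cosmetic, giving its mesh simple collision and jointed physics.
# A large box body covers the root bone, and a smaller cube body sits offset on the pack bone.
# One constraint joins root to pack with twist and second swing locked, leaving only the first swing axis unlocked for hinge-like rotation.
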PhysicsAsset /BRCosmetics/Accessories/FORT_Backpacks/Backpack_F_MED_SmartHyena/Meshes/F_MED_SmartHyena_Pack_Physics1
bodies {
 pack [{shapeType=Box, extentX=7.5, extentY=7.5, extentZ=7.5, shapeName=pack_box, center=<0 -10.7334 0>}]
 root [{shapeType=Box, extentX=21.6342, extentY=7.5, extentZ=25.8689, shapeName=root_box, center=<0 0 -2.8369>}]
}
constraints [{bone1Name=root, bone2Name=pack, swing1LimitDegrees=45, swing2LimitDegrees=45, swing2Motion=Locked, twistLimitDegrees=45, twistMotion=Locked}]
