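# Animation Blueprint for the SmartHyena backpack, with idle, idle-break and reactive animations triggered by damage in-game or modifier reactivity in the front end.
# The cached base pose adds rigid-body physics, pack and lamp offsets, and pins the pelvis to the parent mesh's.
# Output blends between ground physics and skydive physics, which adds velocity-opposed force and layered sine-noise tail wobble.
# SkydiveStateMachine and the SoftRotationBone chain are unconnected to the output.
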AnimBlueprint /BRCosmetics/Accessories/FORT_Backpacks/Backpack_F_MED_SmartHyena/Meshes/F_MED_SmartHyena_Pack_AnimBP
target_skeleton=/Game/Accessories/FORT_Backpacks/Backpack_F_MED_SmartHyena/Meshes/F_MED_SmartHyena_Pack_Skeleton
graphs [
 AnimGraph {nodes [
   n1 = "Transform (Modify) Bone - Bone: SoftRotationBone"!(ComponentPose=Local To Component(), bAlphaBoolEnabled:bool=Get bIsSkyDiving(), Translation="0,0,0", Rotation="0,0,25", Scale="1,1,1") -> (Pose) {
    alphaInputType=Bool
    rotationMode=BMM_Additive
    rotationSpace=BCS_ParentBoneSpace
    translationSpace=BCS_ParentBoneSpace
    alphaBoolBlend {blendInTime=0.2, blendOption=Linear, blendOutTime=0.2}
    boneToModify {boneName=SoftRotationBone}
    rotation {roll=25}
   }
   Component To Local(ComponentPose=@n1)
   Input Pose()
   n2 = RigidBody!(
    ComponentPose=Local To Component(LocalPose=Use cached pose 'CachAnim'())
    Alpha
    OverrideWorldGravity:Vector="0,0,100"
    ExternalForce=Get ExtensionForceDiving() -> (ExtensionForceDiving:Vector)
   ) -> (Pose) {
    bClampLinearTranslationLimitToRefPose=true
    bOverrideWorldGravity=true
    overridePhysicsAsset=&/F_MED_SmartHyena_Pack_Physics
    alphaBoolBlend {blendOption=Linear}
    baseBoneRef {boneName=pack}
    componentAppliedLinearAccClamp=<2000 2000 2000>
    componentLinearAccScale=<0.01 0.01 0.01>
    componentLinearVelScale=<0.01 0.01 0.01>
    simSpaceSettings {maxAngularAcceleration=100, maxAngularVelocity=15, maxLinearAcceleration=100, maxLinearVelocity=15, velocityScaleZ=0, worldAlpha=0.25}
   }
   n3 = "Transform (Modify) Bone - Bone: BirdTail3_M"(ComponentPose=@n2, Alpha=1, Translation="0,0,0", Rotation=Get NoisRot1() -> (NoisRot1:Rotator), Scale="1,1,1") {rotationMode=BMM_Additive, rotationSpace=BCS_BoneSpace, alphaBoolBlend {blendOption=Linear}, boneToModify {boneName=BirdTail3_M}}
   n4 = "Transform (Modify) Bone - Bone: BirdTail4_M"(ComponentPose=@n3, Alpha=1, Translation="0,0,0", Rotation=Get NoisRot2() -> (NoisRot2:Rotator), Scale="1,1,1") {rotationMode=BMM_Additive, rotationSpace=BCS_BoneSpace, alphaBoolBlend {blendOption=Linear}, boneToModify {boneName=BirdTail4_M}}
   n5 = "Transform (Modify) Bone - Bone: BirdTail5_M"(ComponentPose=@n4, Alpha=1, Translation="0,0,0", Rotation=Get NoisRot3() -> (NoisRot3:Rotator), Scale="1,1,1") {rotationMode=BMM_Additive, rotationSpace=BCS_BoneSpace, alphaBoolBlend {blendOption=Linear}, boneToModify {boneName=BirdTail5_M}}
   n6 = "Transform (Modify) Bone - Bone: BirdTail6_M"(ComponentPose=@n5, Alpha=1, Translation="0,0,0", Rotation=Get NoisRot4() -> (NoisRot4:Rotator), Scale="1,1,1") {rotationMode=BMM_Additive, rotationSpace=BCS_BoneSpace, alphaBoolBlend {blendOption=Linear}, boneToModify {boneName=BirdTail6_M}}
   n7 = RigidBody!(ComponentPose=Local To Component(LocalPose=Use cached pose 'CachAnim'()), Alpha, ExternalForce) -> (Pose) {
    bClampLinearTranslationLimitToRefPose=true
    overridePhysicsAsset=&F_MED_SmartHyena_Pack_Physics
    alphaBoolBlend {blendOption=Linear}
    baseBoneRef {boneName=pack}
    componentAppliedLinearAccClamp=<2000 2000 2000>
    componentLinearAccScale=<1 1 1>
    componentLinearVelScale=<1 1 1>
    simSpaceSettings {maxAngularAcceleration=300, maxAngularVelocity=15, maxLinearAcceleration=300, maxLinearVelocity=15, worldAlpha=0.75}
   }
   Output Pose(Result=Blend Poses by bool(
     BlendPose_0=Component To Local(ComponentPose=@n6)
     BlendPose_1=Component To Local(ComponentPose=@n7)
     BlendTime_0=1
     BlendTime_1=0.3
     bActiveValue=Get bIsSkyDiving()
    ) {blendType=Cubic}) {layerGroup=None}
   AnimGraphNode_StateMachine_1 = Idle / FE_Reactive() {is=title AnimGraphNode_StateMachine}
   AnimGraphNode_StateMachine_2 = Idle / Reactive() {is=title AnimGraphNode_StateMachine}
   n8 = Get bIsInFrontEnd()
   n9 = RigidBody!(
    ComponentPose=Local To Component(LocalPose=Blend Poses by bool(BlendPose_0=@AnimGraphNode_StateMachine_1, BlendPose_1=@AnimGraphNode_StateMachine_2, bActiveValue=@n8) {blendType=Linear})
    bAlphaBoolEnabled:bool=Get bIsSkyDiving()
    ExternalForce
   ) -> (Pose) {
    alphaInputType=Bool
    bClampLinearTranslationLimitToRefPose=true
    overridePhysicsAsset=&/F_MED_SmartHyena_Pack_Physics1
    alphaBoolBlend {blendInTime=0.2, blendOption=Cubic, blendOutTime=0.5}
    baseBoneRef {boneName=pack}
    componentLinearAccScale=<0.3 0.3 0.3>
    componentLinearVelScale=<0.3 0.3 0.3>
    simSpaceSettings {maxAngularAcceleration=50, maxAngularVelocity=10, maxLinearAcceleration=50, maxLinearVelocity=10, velocityScaleZ=0}
   }
   n10 = "Transform (Modify) Bone - Bone: pack"(ComponentPose=@n9, Translation="0,1,0", Rotation="0,0,0") {
    rotationSpace=BCS_ParentBoneSpace
    scaleMode=BMM_Replace
    translationMode=BMM_Additive
    alphaBoolBlend {blendOption=Linear}
    boneToModify {boneName=pack}
    rotation {pitch=16}
    translation {y=1}
   }
   n11 = "Transform (Modify) Bone - Bone: LightBox_M"!(ComponentPose=@n10, bAlphaBoolEnabled:bool=Get LampOffsetOn() -> (LampOffsetOn:bool), Translation="0,0,0", Rotation="0,0,14", Scale="1,1,1") -> (Pose) {
    alphaInputType=Bool
    rotationMode=BMM_Additive
    alphaBoolBlend {blendInTime=0.2, blendOption=Cubic, blendOutTime=0.3}
    boneToModify {boneName=LightBox_M}
    rotation {roll=14}
   }
   n12 = "Transform (Modify) Bone - Bone: pack"!(ComponentPose=@n11, bAlphaBoolEnabled:bool=Get bIsSkyDiving(), Translation="0.000,-3.617,-6.076", Rotation="-0.000,0.000,14.697", Scale) -> (Pose) {
    alphaInputType=Bool
    rotationMode=BMM_Additive
    rotationSpace=BCS_ParentBoneSpace
    scaleMode=BMM_Replace
    translationMode=BMM_Additive
    translationSpace=BCS_BoneSpace
    alphaBoolBlend {blendInTime=0.2, blendOption=QuadraticInOut, blendOutTime=0.3}
    boneToModify {boneName=pack}
    rotation {roll=14.6969}
    translation {y=-3.617, z=-6.076}
   }
   n13 = "Transform (Modify) Bone - Bone: pelvis"!(
    ComponentPose=@n12
    bAlphaBoolEnabled:bool=True
    Translation=Get PelvisLoc() -> (PelvisLoc:Vector)
    Rotation=Get PelvisRot() -> (PelvisRot:Rotator)
    Scale=Get PelvisScale() -> (PelvisScale:Vector)
   ) -> (Pose) {
    alphaInputType=Bool
    rotationMode=BMM_Replace
    rotationSpace=BCS_WorldSpace
    scaleMode=BMM_Replace
    scaleSpace=BCS_WorldSpace
    translationMode=BMM_Replace
    translationSpace=BCS_WorldSpace
    alphaBoolBlend {blendInTime=0.2, blendOption=Linear, blendOutTime=0.2}
    boneToModify {boneName=pelvis}
    rotation {roll=25}
   }
   Save cached pose 'CachAnim'(Pose=Component To Local(ComponentPose=@n13)) {cacheName=CachAnim}
   AnimGraphNode_StateMachine_0 = SkydiveStateMachine() {is=title AnimGraphNode_StateMachine}
  ]}
 SkydiveStateMachine {
  entry=ground
  path=AnimGraph.AnimGraphNode_StateMachine_0.SkydiveStateMachine
  ids=id
  nodes [
   AnimStateNode_0 = diving mode() {is=title AnimStateNode}
   AnimStateNode_1 = enter glider() {
    is=title AnimStateNode
    stateEntered {guid=#1}
    stateFullyBlended {guid=#2}
    stateLeft {guid=#3}
   }
   AnimStateNode_2 = ground() {
    is=title AnimStateNode
    stateEntered {guid=#4}
    stateFullyBlended {guid=#5}
    stateLeft {guid=#6}
   }
   AnimStateNode_3 = floating() {is=title AnimStateNode}
   AnimStateNode_4 = Ground to floating() {
    is=title AnimStateNode
    stateEntered {guid=#7}
    stateFullyBlended {guid=#8}
    stateLeft {guid=#9}
   }
   AnimStateNode_5 = Transition to ground() {is=title AnimStateNode}
   AnimStateNode_6 = Reactive() {is=title AnimStateNode}
   AnimStateNode_7 = glider() {is=title AnimStateNode}
   AnimStateNode_8 = Glider to floating() {is=title AnimStateNode}
   AnimStateTransitionNode_0 = AnimStateTransitionNode()
   AnimStateTransitionNode_1 = AnimStateTransitionNode()
   AnimStateTransitionNode_10 = AnimStateTransitionNode()
   AnimStateTransitionNode_11 = AnimStateTransitionNode()
   AnimStateTransitionNode_12 = AnimStateTransitionNode()
   AnimStateTransitionNode_13 = AnimStateTransitionNode()
   AnimStateTransitionNode_14 = AnimStateTransitionNode()
   AnimStateTransitionNode_15 = AnimStateTransitionNode()
   AnimStateTransitionNode_17 = AnimStateTransitionNode()
   AnimStateTransitionNode_2 = AnimStateTransitionNode()
   AnimStateTransitionNode_3 = AnimStateTransitionNode() {crossfadeDuration=0.05}
   AnimStateTransitionNode_4 = AnimStateTransitionNode()
   AnimStateTransitionNode_5 = AnimStateTransitionNode()
   AnimStateTransitionNode_6 = AnimStateTransitionNode()
   AnimStateTransitionNode_7 = AnimStateTransitionNode() {bAutomaticRuleBasedOnSequencePlayerInState=true, crossfadeDuration=0.05}
   AnimStateTransitionNode_8 = AnimStateTransitionNode()
   AnimStateTransitionNode_9 = AnimStateTransitionNode()
  ]
  links [
   {
    dst=floating
    rule=AnimGraph.AnimGraphNode_StateMachine_0.SkydiveStateMachine.AnimStateTransitionNode_17.Transition
    src=Glider to floating
    via=AnimStateTransitionNode_17
   }
   {
    dst=floating
    rule=AnimGraph.AnimGraphNode_StateMachine_0.SkydiveStateMachine.AnimStateTransitionNode_2.Transition
    src=Ground to floating
    via=AnimStateTransitionNode_2
   }
   {
    dst=ground
    rule=AnimGraph.AnimGraphNode_StateMachine_0.SkydiveStateMachine.AnimStateTransitionNode_7.Transition
    src=Reactive
    via=AnimStateTransitionNode_7
   }
   {
    dst=ground
    rule=AnimGraph.AnimGraphNode_StateMachine_0.SkydiveStateMachine.AnimStateTransitionNode_6.Transition
    src=Transition to ground
    via=AnimStateTransitionNode_6
   }
   {
    dst=enter glider
    rule=AnimGraph.AnimGraphNode_StateMachine_0.SkydiveStateMachine.AnimStateTransitionNode_13.Transition
    src=diving mode
    via=AnimStateTransitionNode_13
   }
   {
    dst=floating
    rule=AnimGraph.AnimGraphNode_StateMachine_0.SkydiveStateMachine.AnimStateTransitionNode_12.Transition
    src=diving mode
    via=AnimStateTransitionNode_12
   }
   {
    dst=Transition to ground
    rule=AnimGraph.AnimGraphNode_StateMachine_0.SkydiveStateMachine.AnimStateTransitionNode_5.Transition
    src=enter glider
    via=AnimStateTransitionNode_5
   }
   {
    dst=floating
    rule=AnimGraph.AnimGraphNode_StateMachine_0.SkydiveStateMachine.AnimStateTransitionNode_14.Transition
    src=enter glider
    via=AnimStateTransitionNode_14
   }
   {
    dst=glider
    rule=AnimGraph.AnimGraphNode_StateMachine_0.SkydiveStateMachine.AnimStateTransitionNode_8.Transition
    src=enter glider
    via=AnimStateTransitionNode_8
   }
   {
    dst=Transition to ground
    rule=AnimGraph.AnimGraphNode_StateMachine_0.SkydiveStateMachine.AnimStateTransitionNode_4.Transition
    src=floating
    via=AnimStateTransitionNode_4
   }
   {
    dst=diving mode
    rule=AnimGraph.AnimGraphNode_StateMachine_0.SkydiveStateMachine.AnimStateTransitionNode_11.Transition
    src=floating
    via=AnimStateTransitionNode_11
   }
   {
    dst=enter glider
    rule=AnimGraph.AnimGraphNode_StateMachine_0.SkydiveStateMachine.AnimStateTransitionNode_0.Transition
    src=floating
    via=AnimStateTransitionNode_0
   }
   {
    dst=ground
    rule=AnimGraph.AnimGraphNode_StateMachine_0.SkydiveStateMachine.AnimStateTransitionNode_10.Transition
    src=floating
    via=AnimStateTransitionNode_10
   }
   {
    dst=Glider to floating
    rule=AnimGraph.AnimGraphNode_StateMachine_0.SkydiveStateMachine.AnimStateTransitionNode_15.Transition
    src=glider
    via=AnimStateTransitionNode_15
   }
   {
    dst=Transition to ground
    rule=AnimGraph.AnimGraphNode_StateMachine_0.SkydiveStateMachine.AnimStateTransitionNode_9.Transition
    src=glider
    via=AnimStateTransitionNode_9
   }
   {
    dst=Ground to floating
    rule=AnimGraph.AnimGraphNode_StateMachine_0.SkydiveStateMachine.AnimStateTransitionNode_1.Transition
    src=ground
    via=AnimStateTransitionNode_1
   }
   {
    dst=Reactive
    rule=AnimGraph.AnimGraphNode_StateMachine_0.SkydiveStateMachine.AnimStateTransitionNode_3.Transition
    src=ground
    via=AnimStateTransitionNode_3
   }
  ]
 }
 diving mode {
  path=AnimGraph.AnimGraphNode_StateMachine_0.SkydiveStateMachine.AnimStateNode_0.diving mode
  nodes [Output Animation Pose(Result=Sequence Player 'SmartHyena_SkydivingLoop'() {sequence=/Game/Accessories/FORT_Backpacks/Backpack_F_MED_SmartHyena/Animation/SmartHyena_SkydivingLoop})]
 }
 enter glider {
  path=AnimGraph.AnimGraphNode_StateMachine_0.SkydiveStateMachine.AnimStateNode_1.enter glider
  nodes [Output Animation Pose(Result=Sequence Player 'SmartHyena_GlidingToIdle'() {bLoopAnimation=false, sequence=/Game/Accessories/FORT_Backpacks/Backpack_F_MED_SmartHyena/Animation/SmartHyena_GlidingToIdle})]
 }
 ground {
  path=AnimGraph.AnimGraphNode_StateMachine_0.SkydiveStateMachine.AnimStateNode_2.ground
  nodes [Output Animation Pose(Result=Sequence Player 'SmartHyena_Idle'() {sequence=/BRCosmetics/Accessories/FORT_Backpacks/Backpack_F_MED_SmartHyena/Animation/SmartHyena_Idle})]
 }
 floating {
  path=AnimGraph.AnimGraphNode_StateMachine_0.SkydiveStateMachine.AnimStateNode_3.floating
  nodes [Output Animation Pose(Result=Sequence Player 'SmartHyena_GlidingLoop'() {sequence=/Game/Accessories/FORT_Backpacks/Backpack_F_MED_SmartHyena/Animation/SmartHyena_GlidingLoop})]
 }
 Ground to floating {
  path=AnimGraph.AnimGraphNode_StateMachine_0.SkydiveStateMachine.AnimStateNode_4.Ground to floating
  nodes [Output Animation Pose(Result=Sequence Player 'SmartHyena_IdleToGliding'() {bLoopAnimation=false, sequence=/Game/Accessories/FORT_Backpacks/Backpack_F_MED_SmartHyena/Animation/SmartHyena_IdleToGliding})]
 }
 Transition to ground {
  path=AnimGraph.AnimGraphNode_StateMachine_0.SkydiveStateMachine.AnimStateNode_5.Transition to ground
  nodes [
   Sequence Player 'SmartHyena_Reactive'() {bLoopAnimation=false, sequence=/BRCosmetics/Accessories/FORT_Backpacks/Backpack_F_MED_SmartHyena/Animation/SmartHyena_Reactive}
   Output Animation Pose(Result=Sequence Player 'SmartHyena_Idle'() {bLoopAnimation=false, sequence=&SmartHyena_Idle})
  ]
 }
 Reactive {
  path=AnimGraph.AnimGraphNode_StateMachine_0.SkydiveStateMachine.AnimStateNode_6.Reactive
  nodes [Output Animation Pose(Result=Sequence Player 'SmartHyena_Reactive'() {bLoopAnimation=false, sequence=&SmartHyena_Reactive})]
 }
 glider {
  path=AnimGraph.AnimGraphNode_StateMachine_0.SkydiveStateMachine.AnimStateNode_7.glider
  nodes [Output Animation Pose(Result=Sequence Player 'SmartHyena_Idle'() {sequence=&SmartHyena_Idle})]
 }
 Glider to floating {
  path=AnimGraph.AnimGraphNode_StateMachine_0.SkydiveStateMachine.AnimStateNode_8.Glider to floating
  nodes [Output Animation Pose(Result=Sequence Player 'SmartHyena_PerchedToGliding'() {bLoopAnimation=false, sequence=/Game/Accessories/FORT_Backpacks/Backpack_F_MED_SmartHyena/Animation/SmartHyena_PerchedToGliding})]
 }
 Transition {
  path=AnimGraph.AnimGraphNode_StateMachine_0.SkydiveStateMachine.AnimStateTransitionNode_0.Transition
  nodes [
   n1 = Get bIsSkydiveFloating()
   Result(bCanEnterTransition=AND Boolean(A=Get bIsParachuteOpened(), B=NOT Boolean(A=@n1)))
  ]
 }
 Transition {
  path=AnimGraph.AnimGraphNode_StateMachine_0.SkydiveStateMachine.AnimStateTransitionNode_1.Transition
  nodes [Result(bCanEnterTransition=Get bIsSkyDiving())]
 }
 Transition {
  path=AnimGraph.AnimGraphNode_StateMachine_0.SkydiveStateMachine.AnimStateTransitionNode_10.Transition
  nodes [Result(bCanEnterTransition=NOT Boolean(A=Get bIsSkyDiving()))]
 }
 Transition {
  path=AnimGraph.AnimGraphNode_StateMachine_0.SkydiveStateMachine.AnimStateTransitionNode_11.Transition
  nodes [Result(bCanEnterTransition=Get bIsSkydiveDiveMode())]
 }
 Transition {
  path=AnimGraph.AnimGraphNode_StateMachine_0.SkydiveStateMachine.AnimStateTransitionNode_12.Transition
  nodes [
   n1 = Get bIsSkydiveDiveMode()
   Result(bCanEnterTransition=AND Boolean(A=NOT Boolean(A=@n1), B=Get bIsSkydiveFloating()))
  ]
 }
 Transition {
  path=AnimGraph.AnimGraphNode_StateMachine_0.SkydiveStateMachine.AnimStateTransitionNode_13.Transition
  nodes [Result(bCanEnterTransition=Get bIsParachuteOpened())]
 }
 Transition {
  path=AnimGraph.AnimGraphNode_StateMachine_0.SkydiveStateMachine.AnimStateTransitionNode_14.Transition
  nodes [
   n1 = Get bIsParachuteOpened()
   Result(bCanEnterTransition=AND Boolean(A=NOT Boolean(A=@n1), B=Get bIsSkyDiving()))
  ]
 }
 Transition {
  path=AnimGraph.AnimGraphNode_StateMachine_0.SkydiveStateMachine.AnimStateTransitionNode_15.Transition
  nodes [
   n1 = Get bIsParachuteOpened()
   Result(bCanEnterTransition=AND Boolean(A=Get bIsSkyDiving(), B=NOT Boolean(A=@n1)))
  ]
 }
 Transition {
  path=AnimGraph.AnimGraphNode_StateMachine_0.SkydiveStateMachine.AnimStateTransitionNode_17.Transition
  nodes [
   n1 = "Get Relevant Anim Time (Glider to floating)"!() -> (ReturnValue:float) {is=title K2Node_AnimGetter}
   n2 = "Get Relevant Anim Length (Glider to floating)"!() -> (ReturnValue:float) {is=title K2Node_AnimGetter}
   Result(bCanEnterTransition="float > float"(A=float / float(A=@n1, B=@n2), B=0.9))
  ]
 }
 Transition {
  path=AnimGraph.AnimGraphNode_StateMachine_0.SkydiveStateMachine.AnimStateTransitionNode_2.Transition
  nodes [
   n1 = "Get Relevant Anim Time (Transition to floating)"!() -> (ReturnValue:float) {is=title K2Node_AnimGetter}
   n2 = "Get Relevant Anim Length (Transition to floating)"!() -> (ReturnValue:float) {is=title K2Node_AnimGetter}
   Result(bCanEnterTransition="float > float"(A=float / float(A=@n1, B=@n2), B=0.9))
  ]
 }
 Transition {
  path=AnimGraph.AnimGraphNode_StateMachine_0.SkydiveStateMachine.AnimStateTransitionNode_3.Transition
  nodes [Result(bCanEnterTransition=Get TookDamage() -> (TookDamage:bool))]
 }
 Transition {
  path=AnimGraph.AnimGraphNode_StateMachine_0.SkydiveStateMachine.AnimStateTransitionNode_4.Transition
  nodes [Result(bCanEnterTransition=NOT Boolean(A=Get bIsSkyDiving()))]
 }
 Transition {
  path=AnimGraph.AnimGraphNode_StateMachine_0.SkydiveStateMachine.AnimStateTransitionNode_5.Transition
  nodes [
   n1 = Get bIsParachuteOpened()
   n2 = Get bIsSkyDiving()
   Result(bCanEnterTransition=AND Boolean(A=NOT Boolean(A=@n1), B=NOT Boolean(A=@n2)))
  ]
 }
 Transition {
  path=AnimGraph.AnimGraphNode_StateMachine_0.SkydiveStateMachine.AnimStateTransitionNode_6.Transition
  nodes [
   n1 = "Get Relevant Anim Time (Transition to ground)"!() -> (ReturnValue:float) {is=title K2Node_AnimGetter}
   n2 = "Get Relevant Anim Length (Transition to ground)"!() -> (ReturnValue:float) {is=title K2Node_AnimGetter}
   Result(bCanEnterTransition="float > float"(A=float / float(A=@n1, B=@n2), B=0.9))
  ]
 }
 Transition {path=AnimGraph.AnimGraphNode_StateMachine_0.SkydiveStateMachine.AnimStateTransitionNode_7.Transition, nodes [Result()]}
 Transition {
  path=AnimGraph.AnimGraphNode_StateMachine_0.SkydiveStateMachine.AnimStateTransitionNode_8.Transition
  nodes [
   n1 = "Get Relevant Anim Time (enter glider)"!() -> (ReturnValue:float) {is=title K2Node_AnimGetter}
   n2 = "Get Relevant Anim Length (enter glider)"!() -> (ReturnValue:float) {is=title K2Node_AnimGetter}
   Result(bCanEnterTransition="float > float"(A=float / float(A=@n1, B=@n2), B=0.9))
  ]
 }
 Transition {
  path=AnimGraph.AnimGraphNode_StateMachine_0.SkydiveStateMachine.AnimStateTransitionNode_9.Transition
  nodes [
   n1 = Get bIsParachuteOpened()
   n2 = Get bIsSkyDiving()
   Result(bCanEnterTransition=AND Boolean(A=NOT Boolean(A=@n1), B=NOT Boolean(A=@n2)))
  ]
 }
 Idle / FE_Reactive {
  entry=FE_Idle
  path=AnimGraph.AnimGraphNode_StateMachine_1.Idle / FE_Reactive
  ids=id
  nodes [
   AnimStateNode_0 = FE_Idle() {is=title AnimStateNode}
   AnimStateNode_1 = FE_Reactive() {is=title AnimStateNode, stateEntered {guid=#10, notifyName=Entered FE_Reactive_Anim}}
   AnimStateTransitionNode_0 = AnimStateTransitionNode() {crossfadeDuration=0.05}
   AnimStateTransitionNode_1 = AnimStateTransitionNode() {bAutomaticRuleBasedOnSequencePlayerInState=true, crossfadeDuration=0.05}
  ]
  links [
   {
    dst=FE_Reactive
    rule=AnimGraph.AnimGraphNode_StateMachine_1.Idle / FE_Reactive.AnimStateTransitionNode_0.Transition
    src=FE_Idle
    via=AnimStateTransitionNode_0
   }
   {
    dst=FE_Idle
    rule=AnimGraph.AnimGraphNode_StateMachine_1.Idle / FE_Reactive.AnimStateTransitionNode_1.Transition
    src=FE_Reactive
    via=AnimStateTransitionNode_1
   }
  ]
 }
 FE_Idle {
  path=AnimGraph.AnimGraphNode_StateMachine_1.Idle / FE_Reactive.AnimStateNode_0.FE_Idle
  nodes [
   n1 = Random Sequence Player!() -> (Pose:Pose Link Structure) {
    is=title AnimGraphNode_RandomPlayer
    entries [
     {
      chanceToPlay=0.7
      maxLoopCount=0
      maxPlayRate=1
      minLoopCount=0
      minPlayRate=1
      sequence=&SmartHyena_Idle
      blendIn {blendOption=HermiteCubic, blendTime=0.1, customCurve=null}
     }
     {
      chanceToPlay=0.3
      maxLoopCount=0
      maxPlayRate=1
      minLoopCount=0
      minPlayRate=1
      sequence=/BRCosmetics/Accessories/FORT_Backpacks/Backpack_F_MED_SmartHyena/Animation/SmartHyena_Idle_Break
      blendIn {blendOption=HermiteCubic, blendTime=0.1, customCurve=null}
     }
    ]
   }
   Output Animation Pose(Result=@n1)
  ]
 }
 FE_Reactive {
  path=AnimGraph.AnimGraphNode_StateMachine_1.Idle / FE_Reactive.AnimStateNode_1.FE_Reactive
  nodes [Output Animation Pose(Result=Sequence Player 'SmartHyena_Reactive'() {bLoopAnimation=false, sequence=&SmartHyena_Reactive})]
 }
 Transition {
  path=AnimGraph.AnimGraphNode_StateMachine_1.Idle / FE_Reactive.AnimStateTransitionNode_0.Transition
  nodes [Result(bCanEnterTransition=Get Play Reactive Front End Anim() -> (Play Reactive Front End Anim:bool))]
 }
 Transition {path=AnimGraph.AnimGraphNode_StateMachine_1.Idle / FE_Reactive.AnimStateTransitionNode_1.Transition, nodes [Result()]}
 Idle / Reactive {
  entry=Idle
  path=AnimGraph.AnimGraphNode_StateMachine_2.Idle / Reactive
  ids=id
  nodes [
   AnimStateNode_0 = Idle()
   AnimStateNode_1 = Reactive()
   AnimStateTransitionNode_0 = AnimStateTransitionNode() {crossfadeDuration=0}
   AnimStateTransitionNode_1 = AnimStateTransitionNode() {bAutomaticRuleBasedOnSequencePlayerInState=true, crossfadeDuration=0.3}
  ]
  links [
   {
    dst=Reactive
    rule=AnimGraph.AnimGraphNode_StateMachine_2.Idle / Reactive.AnimStateTransitionNode_0.Transition
    src=Idle
    via=AnimStateTransitionNode_0
   }
   {
    dst=Idle
    rule=AnimGraph.AnimGraphNode_StateMachine_2.Idle / Reactive.AnimStateTransitionNode_1.Transition
    src=Reactive
    via=AnimStateTransitionNode_1
   }
  ]
 }
 Idle {
  path=AnimGraph.AnimGraphNode_StateMachine_2.Idle / Reactive.AnimStateNode_0.Idle
  nodes [
   n1 = Random Sequence Player() {entries [
     {
      chanceToPlay=0.6
      maxLoopCount=0
      maxPlayRate=1
      minLoopCount=0
      minPlayRate=1
      sequence=&SmartHyena_Idle
      blendIn {blendOption=HermiteCubic, blendTime=0, customCurve=null}
     }
     {
      chanceToPlay=0.4
      maxLoopCount=0
      maxPlayRate=1
      minLoopCount=0
      minPlayRate=1
      sequence=&SmartHyena_Idle_Break
      blendIn {blendOption=HermiteCubic, blendTime=0, customCurve=null}
     }
    ]}
   Output Animation Pose(Result=@n1)
  ]
 }
 Reactive {
  path=AnimGraph.AnimGraphNode_StateMachine_2.Idle / Reactive.AnimStateNode_1.Reactive
  nodes [Output Animation Pose(Result=Sequence Player 'SmartHyena_Reactive'() {bLoopAnimation=false, sequence=&SmartHyena_Reactive})]
 }
 Transition {
  path=AnimGraph.AnimGraphNode_StateMachine_2.Idle / Reactive.AnimStateTransitionNode_0.Transition
  nodes [Result(bCanEnterTransition=Get TookDamage())]
 }
 Transition {path=AnimGraph.AnimGraphNode_StateMachine_2.Idle / Reactive.AnimStateTransitionNode_1.Transition, nodes [Result()]}
 EventGraph {nodes [
   n1 = Event BlueprintProcessAnimInputEvent()
   > Branch(Condition=OR Boolean(
     A=IsNamed(self=@n1.AnimInputEvent, InEventName=aie.PlayerVitality.HealthDecreased)
     B=IsNamed(self=@n1.AnimInputEvent, InEventName=Noneaie.PlayerVitality.ShieldsDecreased)
    ))
   > Set TookDamage(TookDamage:bool=true) -> (Output_Get:bool)
   Event BlueprintUpdateAnimation()
   > n2 = Sequence!(execute) -> (then_0, then_1, then_2:Exec, then_3:Exec)
   n3 = GetPartSkeletalMeshComponent()
   Is Valid(exec=@n2.then_1, InputObject=@n3)
   n4 = GetAttachParent!(self:Scene Component Object Reference=@n3) -> (ReturnValue:Scene Component Object Reference)
   > Is Valid(InputObject=@n4)
   n5 = GetSocketTransform!(self=@n4, InSocketName=pelvis, TransformSpace=RTS_World) -> (ReturnValue_Location:Vector, ReturnValue_Rotation:Rotator, ReturnValue_Scale:Vector)
   > Set PelvisLoc(PelvisLoc:Vector=@n5) -> (Output_Get:Vector)
   > Set PelvisRot(PelvisRot:Rotator=@n5.ReturnValue_Rotation) -> (Output_Get:Rotator)
   > Set PelvisScale(PelvisScale:Vector=@n5.ReturnValue_Scale) -> (Output_Get:Vector)
   n6 = Branch(execute=@n2.then_2, Condition=OR Boolean(A=Get bIsSkydiveFloating(), B=Get bIsSkydiveDiveMode()))
   > n7 = Sequence()
   n8 = MakeVector(X=30, Y=40, Z=30)
   > Set NoisRot1(NoisRot1:Rotator=Noise by Sin!(amplitude:double=1.2, Offset:Vector, frequency Speed:Vector=@n8, Frequency control:Vector="0.1,1,1", Negative:bool) -> (ReturnValue:Rotator=~) {is=title K2Node_MacroInstance}) -> (Output_Get:Rotator)
   Set NoisRot2(
    execute=@n7.then_1
    NoisRot2:Rotator=Noise by Sin(amplitude=3, Offset="2,2,2", frequency Speed=@n8, Frequency control="0.1,1,1") -> (ReturnValue=~)
   ) -> (Output_Get:Rotator)
   Set NoisRot3(
    execute=@n7.then_2
    NoisRot3:Rotator=Noise by Sin(amplitude=4, Offset="4.5,4.5,4.5", frequency Speed=@n8, Frequency control="0.1,1,1") -> (ReturnValue=~)
   ) -> (Output_Get:Rotator)
   Set NoisRot4(
    execute=@n7.then_3
    NoisRot4:Rotator=Noise by Sin(amplitude=4, Offset="5,5,5", frequency Speed=@n8, Frequency control="0.1,1,1") -> (ReturnValue=~)
   ) -> (Output_Get:Rotator)
   Set NoisRot1(execute=@n6.else, NoisRot1="0, 0, 0")
   > Set NoisRot2(NoisRot2="0, 0, 0")
   > Set NoisRot3(NoisRot3="0, 0, 0")
   > Set NoisRot4(NoisRot4="0, 0, 0")
   Set LampOffsetOn(execute=@n2.then_3, LampOffsetOn:bool="float > float"(A=Get PawnSpeed2D(), B=50)) -> (Output_Get:bool)
   n9 = Get PawnVelocity()
   > Set ExtensionForceDiving(ExtensionForceDiving:Vector=vector * vector(A=NegateVector!(A:Vector=@n9) -> (ReturnValue:Vector), B=Make Literal Float(Value=0.1)) -> (ReturnValue=~)) -> (Output_Get:Vector)
   n10 = Branch(execute=@n2.then_0, Condition=Get bIsSkyDiving())
   > Set NegatePhysicsAlpha(NegatePhysicsAlpha:double=MapRangeClamped(Value=GetCurveValue(CurveName=PhysicsOff), InRangeA=0, InRangeB=1, OutRangeA=1, OutRangeB=0)) -> (Output_Get:double)
   Set NegatePhysicsAlpha(execute=@n10.else, NegatePhysicsAlpha=1)
   Event OnCharacterCustomizationLoaded()
   > n11 = Cast To FortPlayerPawn(Object=TryGetFortPawn())
   > n12 = Cast To B_Athena_PartModifier_Backpack_SmartHyena!(
    execute
    Object=GetModifierForPartType!(self:Fort Player Pawn Object Reference=@n11, PartType:EFortCustomPartType Enum=Backpack) -> (ReturnValue:Custom Character Part Modifier Object Reference)
   ) -> (then, CastFailed, AsB Athena Part Modifier Backpack Smart Hyena:B Athena Part Modifier Backpack Smart Hyena Object Reference) {is=title K2Node_DynamicCast}
   > n13 = Set Reactivity Enabled(Reactivity Enabled:bool=Get ReactivityEnabled!(self:B Athena Part Modifier Backpack Smart Hyena Object Reference=@n12) -> (ReactivityEnabled:bool)) -> (Output_Get:bool)
   > n14 = Branch(Condition=@n13)
   > n15 = Do Once(Reset=@n16.then)
   Set Play Reactive Front End Anim(execute=[@n17.then, @n15.Completed], Play Reactive Front End Anim:bool=true) -> (Output_Get:bool)
   n16 = Set Play Reactive Front End Anim(execute=@n14.else, Play Reactive Front End Anim=false)
   AnimNotify_Entered FE_Reactive_Anim() {is=title K2Node_Event}
   > Set Play Reactive Front End Anim(Play Reactive Front End Anim=false)
   > n17 = Delay(Duration=8)
   Event BlueprintPostEvaluateAnimation() {is=title K2Node_Event}
   > Set TookDamage(TookDamage=false)
   TryGetPawnOwner()
  ]}
 NoiseBySin {nodes [
   n1 = Inputs!() -> (amplitude:double=15, Offset:Vector=~, frequency Speed:Vector="8,8,8", Frequency control:Vector="0.1,1,1", Negative:bool=~)
   n2 = BreakVector(InVec=@n1.Offset)
   n3 = BreakVector(InVec=@n1.frequency Speed)
   n4 = GetRealTimeSeconds()
   n5 = float * float(A=@n4, B=@n3.Z) {is=title K2Node_PromotableOperator}
   n6 = BreakVector(InVec=@n1.Frequency control)
   n7 = float * float(A="Sin (Radians)"(A=float + float(A=@n5, B=@n2.Z)), B=@n6.Z)
   n8 = float * float(A=@n4, B=@n3.Y)
   n9 = float * float(A="Sin (Radians)"(A=float + float(A=@n8, B=@n2.Y)), B=@n6.Y)
   n10 = float * float(A=@n4, B=@n3)
   n11 = float * float(A="Sin (Radians)"(A=float + float(A=@n10, B=@n2)), B=@n6)
   n12 = Make Literal Float(Value=@n1)
   n13 = SelectFloat(A=-1, B=1, bPickA=@n1.Negative)
   Outputs!(ReturnValue:Rotator="Scale Rotator (Float)"(A:Rotator=MakeRotator(Roll=@n11, Pitch=@n9, Yaw=@n7), B:float=float * float(A=@n12, B=@n13)) -> (ReturnValue:Rotator=~) {is=title K2Node_PromotableOperator})
  ]}
]
variables [
 CurrentHealth {type=number}
 ExtensionForceDiving {type=Vector}
 FE_Reactive {type=boolean}
 IsStaffPickaxe {type=boolean}
 LampOffsetOn {type=boolean}
 NegatePhysicsAlpha {type=number}
 NoisRot1 {type=Rotator}
 NoisRot2 {type=Rotator}
 NoisRot3 {type=Rotator}
 NoisRot4 {type=Rotator}
 Pawn {type=FortPlayerPawn}
 PelvisLoc {type=Vector}
 PelvisRot {type=Rotator}
 PelvisScale {type=Vector}
 Play Reactive Front End Anim {type=boolean}
 PreviousHealth {type=number}
 Reactivity Enabled {type=boolean}
 SoftRotation {type=Rotator}
 TookDamage {type=boolean}
]
